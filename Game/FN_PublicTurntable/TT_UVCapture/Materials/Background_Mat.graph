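# Unlit translucent material with depth test disabled that draws a procedural grid background, apparently for the turntable UV capture tool.
# Grid lines come from thresholding the fractional part of offset, scaled UVs on each axis.
# Outer edge lines come from thresholding the absolute sine of half-tiled UVs, marking the UV tile borders.
# Two chained lerps layer grid and edge colors over the background color, with colors, line count and thicknesses as parameters.
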Material /Game/FN_PublicTurntable/TT_UVCapture/Materials/Background_Mat
graph {
 properties {
  bDisableDepthTest=true
  bEnableResponsiveAA=true
  bFullyRough=true
  bTangentSpaceNormal=false
  bUsedWithMeshDeformer=true
  bUsedWithSkeletalMesh=true
  bUsedWithStaticLighting=true
  blendMode=BLEND_Translucent
  pixelDepthOffsetMode=PDOM_Legacy
  shadingModel=MSM_Unlit
  usedShadingModels=Unlit
 }
 nodes [
  n1 = Subtract(A=TextureCoordinate(), B=Multiply(A=ScalarParameter() {defaultValue=10, parameterName=NumLines}) {constB=0.5})
  n2 = ScalarParameter() {defaultValue=10, parameterName=NumLines}
  n3 = MaterialFunctionCall(Float2=Frac(None=Multiply(A=@n1, B=@n2) {constB=3})) {materialFunction=&BreakOutFloat2Components}
  n4 = OneMinus(None=ScalarParameter() {defaultValue=0.05, parameterName=GridLineThickness})
  n5 = Constant() {r=1}
  n6 = Constant()
  n7 = Add(A=If(A=@n3.R, B=@n4, "A > B"=@n5, "A == B"=@n5, "A < B"=@n6), B=If(A=@n3.G, B=@n4, "A > B"=@n5, "A == B"=@n5, "A < B"=@n6))
  n8 = MaterialFunctionCall(Float2=TextureCoordinate() {uTiling=0.5, vTiling=0.5}) {materialFunction=&BreakOutFloat2Components}
  n9 = ScalarParameter() {defaultValue=0.02, parameterName=outerLineThickness}
  n10 = Constant()
  n11 = Constant() {r=1}
  n12 = If(A=Abs(None=Sine(None=@n8.R)), B=@n9, "A > B"=@n10, "A == B"=@n10, "A < B"=@n11)
  n13 = MaterialFunctionCall(Float2=TextureCoordinate() {uTiling=0.5, vTiling=0.5}) {materialFunction=&BreakOutFloat2Components}
  n14 = If(A=Abs(None=Sine(None=@n13.G)), B=@n9, "A > B"=@n10, "A == B"=@n10, "A < B"=@n11)
  LinearInterpolate(
   A=LinearInterpolate(
    A=VectorParameter() {parameterName=Color_Background, defaultValue=<rgba 0.005 0.005 0.005 1>}.RGB
    B=VectorParameter() {parameterName=Color_GridLines, defaultValue=<rgba 0.007 0.007 0.007 1>}.RGB
    Alpha=Saturate(None=@n7)
   )
   B=VectorParameter() {parameterName=Color_EdgeLines, defaultValue=<rgba 0.2344 0.2344 0.2344 1>}.RGB
   Alpha=Saturate(None=Add(A=@n12, B=@n14))
  )
 ]
}
