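# BlackOut_Mat is an unlit translucent material in the turntable's UV capture folder, apparently meant to black out what it covers, going by its name.
# It disables the depth test, so it draws regardless of geometry in front of it.
# Its graph holds a single Constant node left at its default value, with no textures or parameters.
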Material /Game/FN_PublicTurntable/TT_UVCapture/Materials/BlackOut_Mat
graph {
 properties {bDisableDepthTest=true, blendMode=BLEND_Translucent, pixelDepthOffsetMode=PDOM_Legacy, shadingModel=MSM_Unlit, usedShadingModels=Unlit}
 nodes [Constant()]
}
