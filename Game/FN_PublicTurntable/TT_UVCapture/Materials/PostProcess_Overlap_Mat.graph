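# Post-process material in the turntable UV capture tools that tints bright scene pixels red, apparently to flag overlaps, as its name suggests.
# It desaturates the post-process input and compares it against the desaturated Color parameter as a brightness threshold.
# An If node outputs red-tinted grayscale above the threshold and the original scene color otherwise.
# The check scalar parameter and a loose constant are unconnected.
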Material /Game/FN_PublicTurntable/TT_UVCapture/Materials/PostProcess_Overlap_Mat
graph {
 properties {
  blendMode=BLEND_Translucent
  blendableLocation=BL_SceneColorAfterDOF
  materialDomain=MD_PostProcess
  pixelDepthOffsetMode=PDOM_Legacy
  shadingModel=MSM_Unlit
  usedShadingModels=Unlit
 }
 nodes [
  Constant()
  n1 = ComponentMask(None=SceneTexture() {sceneTextureId=PPI_PostProcessInput0}.Color) {b=true, g=true, r=true}
  n2 = Desaturation(None=@n1)
  If(
   A=@n2
   B=Desaturation(None=VectorParameter() {parameterName=Color, defaultValue=<rgba 0.2758 0.4005 1 1>}.RGB)
   "A > B"=Multiply(A=@n2, B=Constant3Vector() {constant=<rgba 0.6927 0.0399 0 0>})
   "A == B"=@n1
   "A < B"=@n1
  )
  ScalarParameter() {defaultValue=0.55, parameterName=check}
 ]
}
