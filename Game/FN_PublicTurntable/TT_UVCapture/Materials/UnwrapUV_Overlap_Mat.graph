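# Unlit additive material for UV capture in the turntable tool, built to flatten meshes, including skeletal ones, into their UV layout and reveal overlaps.
# UV coordinates are centered, scaled by Mult and Size and padded to 3D, then local position is subtracted to form a vertex offset weighted by a blend parameter.
# Additive blending with depth test disabled lets stacked faces accumulate brightness, with Color and Opacity parameters for tinting.
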Material /Game/FN_PublicTurntable/TT_UVCapture/Materials/UnwrapUV_Overlap_Mat
graph {
 properties {
  bDisableDepthTest=true
  bFullyRough=true
  bTangentSpaceNormal=false
  bUseTranslucencyVertexFog=false
  bUsedWithMeshDeformer=true
  bUsedWithSkeletalMesh=true
  bUsedWithStaticLighting=true
  blendMode=BLEND_Additive
  pixelDepthOffsetMode=PDOM_Legacy
  shadingModel=MSM_Unlit
  translucencyLightingMode=TLM_SurfacePerPixelLighting
  usedShadingModels=Unlit
 }
 nodes [
  Constant() {r=1}
  n1 = TextureCoordinate()
  n2 = ScalarParameter() {defaultValue=0.5, parameterName=sub}
  n3 = Multiply(
   A=Multiply(A=Subtract(A=@n1, B=@n2), B=ScalarParameter() {defaultValue=2, parameterName=Mult})
   B=ScalarParameter() {defaultValue=512, parameterName=Size}
  )
  n4 = Constant()
  LinearInterpolate(
   B=Subtract(
    A=AppendVector(A=@n3, B=@n4)
    B=MaterialFunctionCall() {materialFunction=&LocalPosition}.Local Position
   ) {constA=0}
   Alpha=ScalarParameter() {defaultValue=1, parameterName=blend}
  )
  ScalarParameter() {defaultValue=0.5, parameterName=Opacity}
  VectorParameter() {parameterName=Color, defaultValue=<rgba 0.2758 0.4005 1 1>}
 ]
}
