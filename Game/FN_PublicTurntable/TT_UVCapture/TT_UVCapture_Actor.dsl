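; Turntable capture actor that screenshots a mesh's unwrapped UVs in wireframe or overlap mode.
; SetMesh swaps in a static or skeletal component; SetRenderMode overrides all its materials and enables post-process only for overlap.
; A square-aspect camera, spawned from the Camera component's settings, takes automation screenshots with TAA, tonemapper and MultipassTurntableActor post-process temporarily off.
; OnScreenshotCompleted fires afterwards; a TEST event chains both modes, then quits.
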(event Custom|TEST_Screenshot
  ;[1 0 0] TEST
  (Default|BindEventtoOnScreenshotCompleted (delegate Custom|TEST_handleFirstScreenshot))
  (CallFunction|TakeScreenshots "NewEnumerator0"))

(event Custom|TEST_handleFirstScreenshot
  ;[1 0 0] TEST
  (CallFunction|TakeScreenshots "NewEnumerator1")
  (Default|UnbindallEventsfromOnScreenshotCompleted)
  (Default|BindEventtoOnScreenshotCompleted (delegate Custom|TEST_handleSecondScreenshot)))

(event Custom|TEST_handleSecondScreenshot
  ;[1 0 0] TEST
  (Development|ExecuteConsoleCommand "quit"))

(event EventBeginPlay
  (CallFunction|SpawnCamera))

(event Custom|takeScreenshots
  (params (type :type TT_UVCaptureMode_Enum%20Enum :default "NewEnumerator0") (name :type String))
  (CallFunction|SetRenderMode type)
  (Game|Player|SetViewTargetwithBlend (Game|GetPlayerController 0) (Variables|Default|GetmyCameraActor) 0.0 "VTBlend_Linear" 0.0 false)
  (Utilities|FlowControl|Delay 1.0)
  (CallFunction|PrepareScreenshot)
  (Automation|TakeAutomationScreenshotatCamera (Variables|Default|GetmyCameraActor) name "" (Automation|GetDefaultScreenshotOptionsforRendering "Low"))
  (CallFunction|CleanUpPostScreenshot)
  (Default|CallOnScreenshotCompleted))

(fn cleanUpPostScreenshot
  (Development|ExecuteConsoleCommand "ShowFlag.TemporalAA 2")
  (Development|ExecuteConsoleCommand "ShowFlag.Tonemapper 2")
  (bind _outactors (Actor|GetAllActorsOfClass "/Game/FN_PublicTurntable/MultipassTurntable/MultipassTurntableActor.MultipassTurntableActor_C"))
  (for _array_element _outactors
    (Variables|/Script/Engine.PostProcessComponent|SetbEnabled true (Variables|/Game/FN_PublicTurntable/MultipassTurntable/MultipassTurntableActor.MultipassTurntableActor_C|GetPostProcess _array_element))))

(fn prepareScreenshot
  (Development|ExecuteConsoleCommand "ShowFlag.TemporalAA  0")
  (Development|ExecuteConsoleCommand "ShowFlag.Tonemapper 0")
  (Development|ExecuteConsoleCommand "r.textureStreaming 0")
  (bind _outactors (Actor|GetAllActorsOfClass "/Game/FN_PublicTurntable/MultipassTurntable/MultipassTurntableActor.MultipassTurntableActor_C"))
  (for _array_element _outactors
    (Variables|/Script/Engine.PostProcessComponent|SetbEnabled false (Variables|/Game/FN_PublicTurntable/MultipassTurntable/MultipassTurntableActor.MultipassTurntableActor_C|GetPostProcess _array_element))))

(fn setMesh
  (params (newMesh :type Object%20Reference))
  (Utilities|FlowControl|Sequence
    (:then_0
      (Utilities|IsValid (Variables|Default|GetcurrentPrimitiveComponent)
        (:Is%20Valid
          (Components|DestroyComponent (Variables|Default|GetcurrentPrimitiveComponent)))))
    (:then_1
      (bind _asstatic_mesh (Utilities|Casting|CastToStaticMesh (Variables|GetnewMesh))
        (:then
          (bind _returnvalue (AddComponent|NODE_AddStaticMeshComponent-0 :RelativeTransform_Location "0, 0, 0" :RelativeTransform_Rotation "0, 0, 0" :RelativeTransform_Scale "1.000000,1.000000,1.000000"))
          (Components|StaticMesh|SetStaticMesh _returnvalue _asstatic_mesh)
          (Variables|Default|SetcurrentPrimitiveComponent _returnvalue))
        (:CastFailed
          (bind _asskeletal_mesh (Utilities|Casting|CastToSkeletalMesh (Variables|GetnewMesh)))
          (bind _returnvalue_2 (AddComponent|NODE_AddSkeletalMeshComponent-0 :RelativeTransform_Location "0, 0, 0" :RelativeTransform_Rotation "0, 0, 0" :RelativeTransform_Scale "1.000000,1.000000,1.000000"))
          (Components|SkinnedMesh|SetSkinnedAssetandUpdate _returnvalue_2 _asskeletal_mesh true)
          (Variables|Default|SetcurrentPrimitiveComponent _returnvalue_2))))
    (:then_2
      (Rendering|SetTranslucentSortPriority (Variables|Default|GetcurrentPrimitiveComponent) 2222))))

(fn setRenderMode
  (params (mode :type TT_UVCaptureMode_Enum%20Enum :default "NewEnumerator0"))
  (Variables|Default|SetcurrentRenderMode mode)
  (Utilities|FlowControl|Sequence
    (:then_0
      (Variables|/Script/Engine.PostProcessComponent|SetbEnabled (Utilities|Enum|Equal(Enum) (Variables|Default|GetcurrentRenderMode) "NewEnumerator1") (Variables|Default|GetPostProcess)))
    (:then_1
      (for _index (range (Rendering|Material|GetNumMaterials (Variables|Default|GetcurrentPrimitiveComponent)))
        (Rendering|Material|SetMaterial (Variables|Default|GetcurrentPrimitiveComponent) _index (Utilities|Select :Index (Variables|Default|GetcurrentRenderMode) :NewEnumerator0 "/Game/FN_PublicTurntable/TT_UVCapture/Materials/UnwrapUV_Wireframe_Mat.UnwrapUV_Wireframe_Mat" :NewEnumerator1 "/Game/FN_PublicTurntable/TT_UVCapture/Materials/UnwrapUV_Overlap_Mat.UnwrapUV_Overlap_Mat"))))))

(fn spawnCamera
  (Utilities|FlowControl|Sequence
    (:then_0
      (bind _returnvalue (Game|SpawnActorfromClass "/Script/Engine.CameraActor" (Transformation|GetWorldTransform (Variables|Default|GetCamera)) "AlwaysSpawn" "MultiplyWithRoot" self 0))
      (Variables|Default|SetmyCameraActor _returnvalue)
      (Camera|SetProjectionMode (Variables|/Script/Engine.CameraActor|GetCameraComponent _returnvalue) (Variables|/Script/Engine.CameraComponent|GetProjectionMode (Variables|Default|GetCamera)))
      (Camera|SetOrthoWidth (Variables|/Script/Engine.CameraActor|GetCameraComponent _returnvalue) (Variables|/Script/Engine.CameraComponent|GetOrthoWidth (Variables|Default|GetCamera)))
      (Camera|SetAspectRatio (Variables|/Script/Engine.CameraActor|GetCameraComponent _returnvalue) 1.0))
    (:then_1
      (Transformation|AttachActorToComponent (Variables|Default|GetmyCameraActor) (Variables|Default|GetCamera) "None" "KeepWorld" "KeepWorld" "KeepWorld" true))
    (:then_2
      (return (Variables|Default|GetmyCameraActor)))))

(fn ConstructionScript)
